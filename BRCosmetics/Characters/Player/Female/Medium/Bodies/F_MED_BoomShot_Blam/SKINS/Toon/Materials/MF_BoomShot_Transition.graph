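# Material function for the BoomShot Toon character skin that wipes incoming material attributes to a glowing two-color effect through a halftone-style dot pattern.
# The dots use a rotated, tiled distance gradient on a camera-facing virtual plane at the actor, with FOV and camera-distance compensation.
# MaterialCurveA shifts a local-height sine gradient, thresholded against the dots to form the transition mask.
# It outputs the blended attributes and a Mask.
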MaterialFunction /BRCosmetics/Characters/Player/Female/Medium/Bodies/F_MED_BoomShot_Blam/SKINS/Toon/Materials/MF_BoomShot_Transition
graph {nodes [
  Constant() {r=100}
  FunctionOutput(None=NamedRerouteUsage())
  FunctionOutput(None=NamedRerouteUsage()) {outputName=Mask, sortPriority=1}
  MaterialFunctionCall!(Fade Offset=Constant() {r=10}, Fade Distance=Constant() {r=400}) {materialFunction=/Game/Global/Materials/Character_Parent_2023/Anime_Parent_2023/MaterialFunctions/MFC_DistanceFade_FOV_compensation}
  NamedRerouteDeclaration(None=FunctionInput() {bUsePreviewValueAsDefault=true, inputType=FunctionInput_MaterialAttributes}) {name=InMaterialAttributes, nodeColor=<rgba 0.1953 1 0 1>}
  n1 = Constant()
  n2 = SetMaterialAttributes!(
   MaterialAttributes=GetMaterialAttributes(None=NamedRerouteUsage()) {attributeGetTypes [#1, #2]}.MaterialAttributes
   EmissiveColor=NamedRerouteUsage()
   BaseColor=@n1
   ShadingModel=ShadingModel()
   Roughness=Constant() {r=1}
   Metallic=@n1
   Specular=@n1
  ) {attributeSetTypes [
    #1
    #2
    #3
    #4
    #5
    #6
   ]}
  NamedRerouteDeclaration(None=BlendMaterialAttributes(A=NamedRerouteUsage(), B=@n2, Alpha=NamedRerouteUsage())) {name=Out, nodeColor=<rgba 1 0 0.5156 1>}
  n3 = MaterialFunctionCall() {materialFunction=&CameraDirectionVector}
  n4 = Multiply(A=Constant3Vector() {constant=<rgba 1 1 0 0>}, B=@n3.Result)
  n5 = Add(A=Multiply(B=Normalize(VectorInput=@n4)) {constA=0.25}, B=@n3.Result)
  n6 = Multiply(A=@n3.Result) {constB=50}
  n7 = Add(A=ActorPositionWS(), B=Constant3Vector() {constant=<rgba 0 0 90 1>})
  n8 = Constant() {r=64}
  n9 = Constant() {r=128}
  n10 = InverseLinearInterpolate(
   A=Constant() {r=45}
   B=Constant() {r=90}
   Value=MaterialFunctionCall() {materialFunction=/Engine/Functions/Engine_MaterialFunctions02/Utility/FOV}.FOV angle in degrees
  )
  NamedRerouteDeclaration(None=MaterialFunctionCall(Plane Normal Axis=Normalize(VectorInput=@n5), Plane Center=Add(A=@n6, B=@n7), UV Scale=LinearInterpolate(A=@n8, B=@n9, Alpha=@n10)) {materialFunction=&VirtualPlaneCoordinates}.UVs) {name=ObjectAlighedScreenPosition, nodeColor=<rgba 1 0.867 0 1>}
  n11 = Subtract(A=Distance(A=ActorPositionWS(), B=CameraPositionWS())) {constB=100}
  NamedRerouteDeclaration(None=Saturate(None=Divide(A=@n11) {constB=400})) {name=DistanceBlending, nodeColor=<rgba 0.4766 1 0 1>}
  n12 = NamedRerouteUsage()
  n13 = ScalarParameter() {defaultValue=25, group=BoomShot, parameterName=DotsAmount, sliderMax=512, sortPriority=20}
  n14 = LinearInterpolate(A=@n13, B=Multiply(A=@n13) {constB=0.5}, Alpha=NamedRerouteUsage())
  n15 = Frac(None=Rotator(Coordinate=Multiply(A=@n12, B=@n14), Time=Constant() {r=1}) {speed=0.75})
  NamedRerouteDeclaration(None=Distance(A=Add(A=@n15) {constB=-0.5}, B=Constant())) {name=DotsGradient, nodeColor=<rgba 1 0.0938 0 1>}
  n16 = NamedRerouteUsage()
  n17 = Add(
   A=LocalPosition() {localOrigin=InstancePreSkinning}.Z
   B=LinearInterpolate(Alpha=ScalarParameter() {group=BoomShot, parameterName=MaterialCurveA}) {constA=350, constB=100}
  ) {constB=0}
  n18 = Multiply(A=Sine(None=Divide(A=@n17) {constB=700})) {constB=1.1}
  n19 = ShaderStageSwitch(PixelShader=VertexInterpolator(VS=@n18).PS, VertexShader=@n18)
  NamedRerouteDeclaration(None=Step(X=SmoothStep(Min=@n16, Value=@n19)) {constY=0.2}) {name=TransitionMask, nodeColor=<rgba 1 0 0.4219 1>}
  n20 = Add(A=LocalPosition() {localOrigin=InstancePreSkinning}.Z) {constB=-50}
  n21 = OneMinus(None=VertexInterpolator(VS=Divide(A=@n20) {constB=150}).PS)
  NamedRerouteDeclaration(None=LinearInterpolate(
    A=VectorParameter() {group=BoomShot, parameterName=WipeColor, defaultValue=<rgba 5 0.0065 2.647 1>}.RGB
    B=VectorParameter() {group=BoomShot, parameterName=WipeColorBottom, defaultValue=<rgba 2 0.0026 1.0588 1>}.RGB
    Alpha=Saturate(None=@n21)
   )) {name=EffectColor, nodeColor=<rgba 0 0.7422 1 1>}
 ]}
