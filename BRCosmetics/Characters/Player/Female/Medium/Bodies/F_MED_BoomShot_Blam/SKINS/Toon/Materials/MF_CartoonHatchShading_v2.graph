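# Material function that gives the Toon skin of BoomShot Blam cartoon hatch shading, baked into emissive with black base color.
# Anime light direction dotted with blended pixel/vertex normals drives height lerps between crosshatch and halftone textures, brightening and darkening incoming base color.
# Both textures project onto a camera-facing, FOV-scaled virtual plane, with crosshatch UVs jumping in timed steps.
# An optional head mask adjusts light and dots; time of day scales emissive.
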MaterialFunction /BRCosmetics/Characters/Player/Female/Medium/Bodies/F_MED_BoomShot_Blam/SKINS/Toon/Materials/MF_CartoonHatchShading_v2
graph {nodes [
  CollectionParameter() {
   collection=&FortniteMaterialParameters
   parameterName=SunlightDirectionalVector
  }
  n1 = Constant()
  n2 = GetMaterialAttributes(None=FunctionInput() {inputName=Material, inputType=FunctionInput_MaterialAttributes}) {attributeGetTypes [#1, #2]}
  n3 = ComponentMask(None=MaterialFunctionCall!(In=MaterialFunctionCall() {materialFunction=/Game/Global/Materials/Character_Parent_2023/Anime_Parent_2023/MaterialFunctions/MFC_AnimeShader_LightDirection}.Direction Vector) {materialFunction=/Game/Packages/Fortress_SharedMaterials/Cinematics/Anime/MF_CINE_AnimeLightDirOverride}.Result) {b=true, g=true, r=true}
  n4 = MaterialFunctionCall(Float3=@n3) {materialFunction=&BreakOutFloat3Components}
  n5 = ScalarParameter() {
   desc="0.0 = the light vector is locked on the XY plane.\r\n1.0 = the actual light vector is used  \r\n(for Faces use a blend value of 0,0)"
   group=Cartoon
   parameterName=LockedZLightVector_Blend
   sliderMax=1
  }
  n6 = Normalize(VectorInput=MaterialFunctionCall(X=@n4.R, Y=@n4.R, Z=Multiply(A=@n4.B, B=@n5)) {materialFunction=&MakeFloat3}.Result)
  n7 = NamedRerouteUsage()
  n8 = ComponentMask(None=PixelNormalWS()) {b=true, g=true, r=true}
  n9 = ComponentMask(None=VertexNormalWS()) {b=true, g=true, r=true}
  n10 = ScalarParameter() {defaultValue=0.25, group=Cartoon, parameterName=CRTN_Pixel-Vertex}
  n11 = DotProduct(
   A=StaticSwitchParameter(True=LinearInterpolate(A=@n3, B=@n6, Alpha=@n7), False=@n3) {group=Cartoon, parameterName=CRTN_UseHeadMask}
   B=ShadingPathSwitch(Default=LinearInterpolate(A=@n8, B=@n9, Alpha=@n10), Forward=@n8, Mobile=@n8)
  )
  n12 = Saturate(None=Divide(A=Subtract(A=@n11) {constB=-1}, B=ScalarParameter() {defaultValue=1.25, group=Cartoon, parameterName=CRTN_LightLevels}))
  n13 = OneMinus(None=Power(Base=@n12, Exp=ScalarParameter() {defaultValue=5, group=Cartoon, parameterName=CRTN_LightPower}))
  n14 = ScalarParameter() {defaultValue=8.3, group=Cartoon, parameterName=CRTN_Crosshatch-OffsetUV}
  n15 = Multiply(A=ScalarParameter() {defaultValue=0.46, group=Cartoon, parameterName=CRTN_Crosshatch-OffsetSpeed}, B=Time())
  n16 = Floor(None=Multiply(A=ScalarParameter() {defaultValue=12, group=Cartoon, parameterName=CRTN_Crosshatch-NumberOffsets}, B=Frac(None=@n15)))
  n17 = ScalarParameter() {defaultValue=0.01, group=Cartoon, parameterName=CRTN_Crosshatch-Scale}
  n18 = Constant() {desc=Needs at least a small amount of this or it will break, r=0.25}
  n19 = MaterialFunctionCall() {materialFunction=&CameraDirectionVector}
  n20 = Normalize(VectorInput=Multiply(A=Constant3Vector() {constant=<rgba 1 1 0 0>}, B=@n19.Result))
  n21 = ScalarParameter() {defaultValue=70, group=Cartoon, parameterName=CRTN_VP-CameraVectorMin}
  n22 = ObjectPositionWS()
  n23 = Add(A=ActorPositionWS(), B=VectorParameter() {group=Cartoon, parameterName=CRTN_VP-ActorPosOffset, defaultValue=<rgba 0 0 80 1>}.RGB)
  n24 = MaterialFunctionCall(
   Input=MaterialFunctionCall() {materialFunction=/Engine/Functions/Engine_MaterialFunctions02/Utility/FOV}."FOV "
   Input Low=Constant() {r=0.414}
   Input High=Constant() {r=1.091}
   Target Low=Constant()
   Target High=Constant() {r=1}
  ) {materialFunction=&RemapValueRange}
  n25 = MaterialFunctionCall(
   Plane Normal Axis=Add(A=Multiply(A=@n18, B=@n20), B=@n19.Result)
   Plane Center=Add(A=Multiply(A=@n19.Result, B=@n21), B=StaticSwitchParameter(True=@n22, False=@n23) {group=Cartoon, parameterName=CRTN_UseObjectPos})
   UV Scale=LinearInterpolate(
    A=ScalarParameter() {defaultValue=1.4, group=Cartoon, parameterName=CRTN_VP-Scale-LowFOV}
    B=ScalarParameter() {defaultValue=1.75, group=Cartoon, parameterName=CRTN_VP-Scale-HighFOV}
    Alpha=Saturate(None=@n24.Result)
   )
  ) {materialFunction=&VirtualPlaneCoordinates}
  n26 = TextureSampleParameter2D(UVs=Add(A=Multiply(A=@n14, B=@n16), B=Multiply(A=@n17, B=@n25.UVs))) {
   group=Cartoon
   parameterName=CRTN_CrosshatchTexture
   samplerType=SAMPLERTYPE_Masks
   texture=/Game/Global/Textures/Default/Tileable_Patterns/OakPP_Post
  }
  n27 = OneMinus(None=@n26.R)
  n28 = TextureSampleParameter2D(UVs=@n25.UVs) {
   group=Cartoon
   parameterName=CRTN-DotTexture
   samplerType=SAMPLERTYPE_LinearGrayscale
   texture=/Game/Characters/Player/Female/Medium/Bodies/F_MED_TNTina/Materials/HalftoneOffset
  }
  n29 = VertexInterpolator(VS=PreSkinnedPosition())
  n30 = Add(
   A=Divide(A=ComponentMask(None=@n29.PS) {b=true}, B=ScalarParameter() {defaultValue=5, group=Cartoon, parameterName=CRTN_LocalMask-Size})
   B=ScalarParameter() {defaultValue=-28, group=Cartoon, parameterName=CRTN_LocalMask-Add}
  )
  n31 = ScalarParameter() {defaultValue=1, group=Cartoon, parameterName=CRTN_LocalMask-Mod}
  n32 = Multiply(A=Saturate(None=Multiply(A=@n30, B=@n31)), B=ScalarParameter() {defaultValue=1, group=Cartoon, parameterName=CRTN_LocalMask-Blend})
  n33 = Constant()
  n34 = Saturate(None=Add(A=@n28.R, B=StaticSwitchParameter(True=@n32, False=@n33) {group=Cartoon, parameterName=CRTN_UseHeadMask}))
  n35 = Multiply(
   A=Multiply(
    A=MaterialFunctionCall!(Diffuse 1, Diffuse 2, Transistion Phase=@n13, Height Texture 1=@n27, Height Texture 2=@n34, Contrast) {materialFunction=/Engine/Functions/Engine_MaterialFunctions02/Texturing/HeightLerpWithTwoHeightMaps}.Lerp Alpha No Contrast
    B=ScalarParameter() {defaultValue=0.8, group=Cartoon, parameterName=CRTN_BrightsAmount}
   )
   B=@n2.BaseColor
  )
  n36 = Add(
   A=MaterialFunctionCall(Transistion Phase=OneMinus(None=@n12), Height Texture 1=@n27, Height Texture 2=@n34, Contrast=Constant() {r=1}) {materialFunction=&HeightLerpWithTwoHeightMaps}.Alpha
   B=ScalarParameter() {defaultValue=0.2, group=Cartoon, parameterName=CRTN__DarksAmount}
  )
  n37 = FunctionInput(Preview=Constant()) {bUsePreviewValueAsDefault=true, inputName=UseMaskToBrighten, inputType=FunctionInput_Scalar}
  n38 = Multiply(A=ScalarParameter() {defaultValue=0.4, group=Cartoon, parameterName=CRTN_MaskToBrightenAmount}, B=@n37)
  n39 = Multiply(A=Saturate(None=Add(A=@n35, B=@n2.BaseColor)), B=Saturate(None=Add(A=@n36, B=@n38)))
  n40 = MaterialFunctionCall(
   Value_Morning=ScalarParameter() {defaultValue=0.4, group=Cartoon, parameterName=CRTN_EmissiveAmount-Morning}
   Value_Day=ScalarParameter() {defaultValue=0.8, group=Cartoon, parameterName=CRTN_EmissiveAmount-Day}
   Value_Evening=ScalarParameter() {defaultValue=0.3, group=Cartoon, parameterName=CRTN_EmissiveAmount-Eve}
   Value_Night=ScalarParameter() {defaultValue=0.3, group=Cartoon, parameterName=CRTN_EmissiveAmount-Night}
  ) {materialFunction=&Scalar_Time_of_Day}
  n41 = Constant() {r=1}
  n42 = Constant3Vector() {constant=<rgba 1 1 1 0>}
  n43 = MaterialFunctionCall(
   Color_Morning=@n42
   Color_Day=VectorParameter() {group=Cartoon, parameterName=CRTN_ColorShift-Day, defaultValue=<rgba 0.88 0.9821 1.1 1>}.RGB
   Color_Evening=@n42
   Color_Night=@n42
  ) {materialFunction=&Color_Time_of_Day}
  n44 = SetMaterialAttributes!(
   MaterialAttributes=@n2.MaterialAttributes
   BaseColor=@n1
   EmissiveColor=Multiply(A=Multiply(A=@n39, B=@n40.Result), B=LinearInterpolate(A=@n41, B=@n43.Result, Alpha=@n37))
   Specular=@n1
   Roughness=Constant() {r=1}
   Metallic=@n1
   ShadingModel=ShadingModel()
  ) {attributeSetTypes [
    #1
    #2
    #3
    #4
    #5
    #6
   ]}
  FunctionOutput(None=@n44)
  n45 = NamedRerouteDeclaration() {name=PreSkinnedPos, nodeColor=<rgba 0 1 0.9297 1>}
  LinearInterpolate(
   B=NamedRerouteDeclaration(None=Normalize(VectorInput=@n45)) {name=SphericalNormals, nodeColor=<rgba 1 0.7734 0 1>}
   Alpha=ScalarParameter() {defaultValue=0.25, group=06_Thin Film, parameterName=CRTN_SphereNormalBlend, sliderMax=1}
  ) {constAlpha=0.25}
  n46 = SphereMask(
   A=VertexInterpolator(VS=LocalPosition() {localOrigin=InstancePreSkinning}.XYZ).PS
   B=VectorParameter() {group=06_Thin Film, parameterName=SphereMaskOffset, defaultValue=<rgba 0 0 0 1>}.RGB
   Radius=ScalarParameter() {defaultValue=15, parameterName=SphereMaskRadius}
  ) {attenuationRadius=10, hardnessPercent=0}
  NamedRerouteDeclaration(None=Saturate(None=Multiply(A=@n46) {constB=3})) {name=SphereMask, nodeColor=<rgba 0 0.4609 1 1>}
  NamedRerouteUsage()
  NamedRerouteUsage()
  ScalarParameter() {defaultValue=1, group=06_Thin Film, parameterName=CRTN_SphereNormalScale}
  VectorParameter() {group=06_Thin Film, parameterName=SphericalNormalPosOffset, defaultValue=<rgba 0 0 0 1>}
  VertexInterpolator(VS=LocalPosition() {localOrigin=InstancePreSkinning}.XYZ)
 ]}
