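# Parent material for the toon skin of the female BoomShot Blam character body, built on material attributes for skeletal meshes.
# The shared character PreFX function feeds cartoon hatch shading, then a BoomShot transition function, then the shared PostFX function.
# A static switch, UseCharacterCartoonOutline, replaces that result with a cartoon outline function.
# A selectable channel of the FX mask texture, blue by default, drives the outline's thickness.
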Material /BRCosmetics/Characters/Player/Female/Medium/Bodies/F_MED_BoomShot_Blam/SKINS/Toon/Materials/M_BoomShot_Toon_Parent
graph {
 properties {
  bEnableResponsiveAA=true
  bUseMaterialAttributes=true
  bUsedWithMeshDeformer=true
  bUsedWithMorphTargets=true
  bUsedWithSkeletalMesh=true
  pixelDepthOffsetMode=PDOM_Legacy
  subsurfaceProfile=&SS_HeroSkin_02
  usedShadingModels=Default Lit
 }
 nodes [
  n1 = TextureSampleParameter2D() {
   group=Cartoon
   parameterName=FXMask
   samplerType=SAMPLERTYPE_Masks
   texture=/BRCosmetics/Characters/Player/Female/Medium/Bodies/F_MED_BoomShot_Blam/SKINS/Toon/Textures/T_BoomShot_BlamToon_Body_FX
  }
  n2 = MaterialFunctionCall() {materialFunction=&BaseFN_CharacterFunction_PreFX}
  n3 = MaterialFunctionCall(
   AO=@n2.AO
   Distance Blend=@n2.AO
   Fresnel=@n2.AO
   Gmap=@n2.AO
   Pre Skinned Local Position=@n2.AO
   Material=MaterialFunctionCall!(In=MaterialFunctionCall!(UseMaskToBrighten=@n2.Skin, Material=@n2.Skin) {materialFunction=&/MF_CartoonHatchShading_v2}.Result) {materialFunction=&/MF_BoomShot_Transition}.Result
  ) {materialFunction=&BaseFN_CharacterFunction_PostFX}
  StaticSwitchParameter(
   True=MaterialFunctionCall!(ThicknessMap=ChannelMaskParameter(None=@n1.RGB) {group=Cartoon, maskChannel=Blue, parameterName=ThicknessFXMask, defaultValue=<rgba 0 0 1 0>}.R) {materialFunction=&/MF_TnTinaOutlines}.Result
   False=@n3.Result
  ) {group=Cartoon, parameterName=UseCharacterCartoonOutline}
 ]
}
